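# Niagara system for the Relaxed Drummer drum emote that periodically pops out short-lived emoji sprites.
# A looping burst launches particles along world up with drag, each rolling a random VisibilityTag from zero to five alongside six sprite renderers.
# The tag also indexes a float array for a per-emoji base EmojiSize.
# Curves scale sprite size and EmojiSize with an overshoot pop-in and shrink-out, while a color curve tints RGB.
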NiagaraSystem /SparksCosmetics/Emotes/Drum/RelaxedDrummer/Effects/NS_Emote_Drum_RelaxedDrumme_Emojis
dependencies {
 usedDataInterfaces [
  &NiagaraDataInterfaceArrayFloat
  &NiagaraDataInterfaceCurve
  &NiagaraDataInterfaceVector2DCurve
  &NiagaraDataInterfaceColorCurve
 ]
 usedDynamicInputs []
 usedModules [&SystemState]
}
emitters {Emojis {
  emitterUpdateScript [
   &EmitterState(
    Life Cycle Mode=::NewEnumerator1
    Loop Duration=&RandomRangeFloat(Minimum=1.5, Maximum=1.3)
    Recalculate Duration Each Loop=true
    UseLoopDelay=false
   )
   &SpawnBurst_Instantaneous(Use Spawn Probability=false, UseLoopCountLimit=false)
  ]
  particleSpawnScript [
   &InitializeParticle(
    Lifetime=&RandomRangeFloat(Minimum=0.9, Maximum=0.9)
    UsePositionOffset=true
    Write Material Random=true
   )
   &ApplyInitialForces()
   /SetVariables(Particles.VisibilityTag=&RandomRangeInt(Minimum=0, Maximum=5)) {isSetParameters=true}
   /SetVariables(Particles.EmojiSize:Vector2f=&Vector2DFromFloat(Value=&SelectFloatFromArray(Direct Array Index=$Particles.VisibilityTag, Float Selection Array=DataInterface() {propertyValues {FloatData [15,20,21,30,25,25]}}))) {isSetParameters=true}
   &CameraOffset(Camera Offset Amount=10)
   &AddVelocity(Velocity=&TransformVector(Vector=<0 0 200>, Source Space=::World, Destination Space=::Local))
  ]
  particleUpdateScript [
   &ParticleState()
   &ScaleSpriteSize(
    Non-Uniform Curve Sprite Scale=DataInterface() {propertyValues {
      XCurve {keys [
        {arriveTangent=-0.024, interpMode=RCIM_Cubic, leaveTangent=-0.024, leaveTangentWeight=0.3333, tangentMode=RCTM_User, time=0, value=0}
        {
         arriveTangent=-0.0357
         arriveTangentWeight=0.0586
         interpMode=RCIM_Cubic
         leaveTangent=-0.0357
         leaveTangentWeight=0.2747
         tangentMode=RCTM_User
         time=0.1928
         value=1.6834
        }
        {arriveTangent=-0.057, tangentMode=RCTM_User, time=0.4652, value=1.0046}
        {
         arriveTangent=-1.3648
         arriveTangentWeight=0.1776
         interpMode=RCIM_Cubic
         leaveTangent=-1.3648
         leaveTangentWeight=0.0667
         tangentMode=RCTM_User
         time=0.8
         value=1
        }
        {tangentMode=RCTM_User, time=1, value=0}
       ]}
      YCurve {keys [
        {arriveTangent=-0.024, interpMode=RCIM_Cubic, leaveTangent=-0.024, leaveTangentWeight=0.3333, tangentMode=RCTM_User, time=0, value=0}
        {
         arriveTangent=-0.0357
         arriveTangentWeight=0.0586
         interpMode=RCIM_Cubic
         leaveTangent=-0.0357
         leaveTangentWeight=0.2747
         tangentMode=RCTM_User
         time=0.1928
         value=1.5702
        }
        {arriveTangent=-0.057, tangentMode=RCTM_User, time=0.318, value=1.0046}
        {
         arriveTangent=-1.3648
         arriveTangentWeight=0.1776
         interpMode=RCIM_Cubic
         leaveTangent=-1.3648
         leaveTangentWeight=0.0667
         tangentMode=RCTM_User
         time=0.8
         value=1
        }
        {tangentMode=RCTM_User, time=1, value=0}
       ]}
     }}
    Scale Sprite Size Mode=::NewEnumerator3
   )
   &Drag(Drag=5, Use Linear Drag=true, Use Rotational Drag=true)
   &SolveForcesAndVelocity(Clamp Velocity=false, Limit Acceleration=false)
   &ScaleColor(
    Scale RGB=&MakeVectorFromLinearColorRGB(Color=&ColorFromCurve(ColorCurve=DataInterface() {propertyValues {
        AlphaCurve {keys []}
        BlueCurve {keys [<at 0.7035 0 RCIM_Cubic>,<at 1 0.1254 RCIM_Cubic>]}
        GreenCurve {keys [<at 0.7035 0 RCIM_Cubic>,<at 1 0 RCIM_Cubic>]}
        RedCurve {keys [<at 0.7035 0.8385 RCIM_Cubic>,<at 1 0.8385 RCIM_Cubic>]}
       }}))
    ScaleA=true
    ScaleRGB=true
   )
   /SetVariables(Particles.EmojiSize:Vector2f=&Multiply_Vector2D(
     A=$Particles.Initial.EmojiSize
     B=&Vector2DFromCurve(
      Vector2Curve=DataInterface() {propertyValues {
        XCurve {keys [
          {arriveTangent=-0.024, interpMode=RCIM_Cubic, leaveTangent=-0.024, leaveTangentWeight=0.3333, tangentMode=RCTM_User, time=0, value=0}
          {
           arriveTangent=-0.0357
           arriveTangentWeight=0.0643
           interpMode=RCIM_Cubic
           leaveTangent=-0.0357
           leaveTangentWeight=0.2691
           tangentMode=RCTM_User
           time=0.1928
           value=1.6834
          }
          {arriveTangent=-0.057, tangentMode=RCTM_User, time=0.4652, value=1.0046}
          {
           arriveTangent=-1.3648
           arriveTangentWeight=0.1116
           interpMode=RCIM_Cubic
           leaveTangent=-1.3648
           leaveTangentWeight=0.0667
           tangentMode=RCTM_User
           time=0.8
           value=1
          }
          {tangentMode=RCTM_User, time=1, value=0}
         ]}
        YCurve {keys [
          {arriveTangent=-0.024, interpMode=RCIM_Cubic, leaveTangent=-0.024, leaveTangentWeight=0.3333, tangentMode=RCTM_User, time=0, value=0}
          {
           arriveTangent=-0.0357
           arriveTangentWeight=0.0643
           interpMode=RCIM_Cubic
           leaveTangent=-0.0357
           leaveTangentWeight=0.2691
           tangentMode=RCTM_User
           time=0.1928
           value=1.5702
          }
          {arriveTangent=-0.057, tangentMode=RCTM_User, time=0.318, value=1.0046}
          {
           arriveTangent=-1.3648
           arriveTangentWeight=0.1607
           interpMode=RCIM_Cubic
           leaveTangent=-1.3648
           leaveTangentWeight=0.0667
           tangentMode=RCTM_User
           time=0.8
           value=1
          }
          {tangentMode=RCTM_User, time=1, value=0}
         ]}
       }}
      ScaleCurve=<1 1>
     )
    )) {isSetParameters=true}
  ]
  rendererClasses [&NiagaraSpriteRendererProperties]
  renderers [
   {rendererClass=&NiagaraSpriteRendererProperties}
   {rendererClass=&NiagaraSpriteRendererProperties, rendererIndex=1}
   {rendererClass=&NiagaraSpriteRendererProperties, rendererIndex=2}
   {rendererClass=&NiagaraSpriteRendererProperties, rendererIndex=3}
   {rendererClass=&NiagaraSpriteRendererProperties, rendererIndex=4}
   {rendererClass=&NiagaraSpriteRendererProperties, rendererIndex=5}
  ]
 }}
summary {
 systemName=NS_Emote_Drum_RelaxedDrumme_Emojis
 emitters [{emitterName=Emojis, rendererClasses [&NiagaraSpriteRendererProperties]}]
 userVariables []
}
