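# Niagara system that adds a cyan-blue phone glow for the Relaxed Drummer drum emote.
# PhoneGlow renders a scaled mesh, while PhoneLight renders an offset sprite and light, both as burst-spawned particles that persist until deactivation.
# User.GlowScale, default zero, drives mesh Y scale and both emitters' alpha, multiplied by per-emitter time-of-day factors.
# The bIsFrontEndPreview flag raises the light's daytime factor for front-end previews.
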NiagaraSystem /SparksCosmetics/Emotes/Drum/RelaxedDrummer/Effects/NS_Emote_Drum_RelaxedDrummer_PhoneGlow
dependencies {usedDataInterfaces [], usedDynamicInputs [], usedModules [&SystemState]}
emitters {
 PhoneGlow {
  emitterUpdateScript [
   &EmitterState()
   &SpawnBurst_Instantaneous(Use Spawn Probability=false, UseLoopCountLimit=true)
  ]
  particleSpawnScript [
   &InitializeParticle(Color=<rgba 0 1 3 0.2>, Mesh Scale=<0.07 0.3 0.0977>, Mesh Scale Mode=::NewEnumerator4, UsePositionOffset=false, Write Material Random=true)
   &InitialMeshOrientation(Mesh Orientation Mode=::NewEnumerator2, Orientation Vector=<0 0 1>, Rotation=<0.25 0 0>, Use Rotation Vector=true)
  ]
  particleUpdateScript [
   &ParticleState(
    Kill Particles When Lifetime Has Elapsed=false
    Let Infinitely Lived Particles Die When Emitter Deactivates=true
    Loop Particles Lifetime=true
   )
   &ScaleMeshSize(Scale Factor=&MakeVector(X=1, Y=$User.GlowScale, Z=1))
   &ScaleColor(
    Scale Alpha=&Multiply_Float(
     A=$User.GlowScale
     B=&FNTimeOfDayScalar(Night Scale=1, Morning Scale=0.25, Day Scale=0.8, Evening Scale=0.6)
    )
    ScaleA=true
    ScaleRGB=true
   )
  ]
  rendererClasses [&NiagaraMeshRendererProperties]
  renderers [{rendererClass=&NiagaraMeshRendererProperties, rendererIndex=0}]
 }
 PhoneLight {
  emitterUpdateScript [
   &EmitterState()
   &SpawnBurst_Instantaneous(Use Spawn Probability=false, UseLoopCountLimit=true)
  ]
  particleSpawnScript [
   &InitializeParticle(
    Color=<rgba 0 1 3 1>
    Position Offset=<0 20 0>
    Sprite Size Mode=::NewEnumerator3
    Uniform Sprite Size=25
    UsePositionOffset=true
    Write Material Random=true
   )
   /SetVariables(Particles.LightRadius=25) {isSetParameters=true}
  ]
  particleUpdateScript [
   &ParticleState(
    Kill Particles When Lifetime Has Elapsed=false
    Let Infinitely Lived Particles Die When Emitter Deactivates=true
    Loop Particles Lifetime=true
   )
   &ScaleColor(
    Scale Alpha=&Multiply_Float(
     A=$User.GlowScale
     B=&FNTimeOfDayScalar(
      Night Scale=2.25
      Morning Scale=1.5
      Day Scale=&MakeCustomFloatFromBool(Bool=$User.bIsFrontEndPreview, True Float=1.5, False Float=1)
      Evening Scale=1.5
     )
    )
    ScaleA=true
    ScaleRGB=true
   )
  ]
  rendererClasses [&NiagaraLightRendererProperties, &NiagaraSpriteRendererProperties]
  renderers [
   {rendererClass=&NiagaraLightRendererProperties, rendererIndex=0}
   {rendererClass=&NiagaraSpriteRendererProperties, rendererIndex=1}
  ]
 }
}
summary {
 systemName=NS_Emote_Drum_RelaxedDrummer_PhoneGlow
 emitters [
  {emitterName=PhoneGlow, rendererClasses [&NiagaraMeshRendererProperties]}
  {emitterName=PhoneLight, rendererClasses [&NiagaraLightRendererProperties, &NiagaraSpriteRendererProperties]}
 ]
 userVariables [
  User.bIsFrontEndPreview {type=NiagaraBool}
  User.GlowScale {type=NiagaraFloat, defaultValue=0, description=""}
  User.InitialOffset {type=Vector3f, description="", defaultValue {struct=Vector3f, value=<0 0 0>}}
 ]
}
